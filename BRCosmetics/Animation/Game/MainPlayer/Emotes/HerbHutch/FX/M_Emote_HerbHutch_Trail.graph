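# Translucent, unlit, two-sided Niagara ribbon material for the trail effect of the HerbHutch emote.
# Blends three colors across ribbon V, tints by particle color and passes the result through MF_FX_EmissiveAdjustments.
# An opacity mask combines a time-animated sine-wave tail falloff along U with V edge fading, driven by dynamic parameters.
# Particle alpha, depth fade and a view-angle term modulate that mask; several similarly named scalar parameters sit unconnected.
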Material /BRCosmetics/Animation/Game/MainPlayer/Emotes/HerbHutch/FX/M_Emote_HerbHutch_Trail
graph {
 properties {
  bUsedWithNiagaraRibbons=true
  blendMode=BLEND_Translucent
  pixelDepthOffsetMode=PDOM_Legacy
  shadingModel=MSM_Unlit
  twoSided=true
  usedShadingModels=Unlit
 }
 nodes [
  Constant()
  n1 = ScalarParameter() {defaultValue=0.1, parameterName=FresnelExp}
  MaterialFunctionCall(Invert Fresnel=StaticBool() {value=true}, Power=@n1) {materialFunction=&Fresnel_Function}
  n2 = VectorParameter() {parameterName=Color 3, defaultValue=<rgba 0.0835 0.1111 0.2292 1>}
  n3 = TextureCoordinate()
  n4 = ComponentMask(None=@n3) {g=true}
  n5 = OneMinus(None=Add(A=@n4) {constB=0.3})
  n6 = LinearInterpolate(
   A=VectorParameter() {parameterName=Color 1, defaultValue=<rgba 0.375 0.1367 0.1367 1>}.RGB
   B=VectorParameter() {parameterName=Color 2, defaultValue=<rgba 0.3073 0.3073 0.3073 0.2>}.RGB
   Alpha=Saturate(None=Multiply(A=@n5) {constB=6})
  )
  n7 = Saturate(None=Multiply(A=Add(A=@n4) {constB=-0.26}) {constB=6})
  MaterialFunctionCall(Emissive=Multiply(A=LinearInterpolate(A=@n2.RGB, B=@n6, Alpha=@n7), B=ParticleColor().RGB)) {materialFunction=&MF_FX_EmissiveAdjustments}
  n8 = CameraVectorWS()
  n9 = VertexNormalWS()
  n10 = Power(Base=OneMinus(None=DotProduct(A=@n8, B=@n9))) {constExponent=1}
  n11 = ComponentMask(None=@n3) {r=true}
  n12 = OneMinus(None=@n11)
  n13 = DynamicParameter() {defaultValue=<rgba 0.7 0.1 0.1 0.6>, paramNames [TrailWaveSpeed, TrailWaveIntensity, EdgeFade, LengthFade]}
  n14 = Multiply(A=Add(A=@n4, B=Time()), B=@n13.TrailWaveSpeed)
  n15 = Multiply(A=ConstantBiasScale(None=Sine(None=@n14)), B=@n13.TrailWaveIntensity)
  n16 = Power(Base=Multiply(A=@n12, B=Subtract(A=@n12, B=@n15))) {constExponent=1.5}
  n17 = DynamicParameter() {defaultValue=<rgba 0.7 0.1 0.1 0.6>, paramNames [TrailWaveSpeed, TrailWaveIntensity, EdgeFade, LengthFade]}
  n18 = OneMinus(None=Abs(None=ConstantBiasScale(None=@n4) {bias=-0.5, scale=2}))
  n19 = Multiply(A=Saturate(None=Divide(A=@n16, B=@n17.LengthFade)), B=Saturate(None=Divide(A=@n18, B=@n17.EdgeFade)))
  n20 = ParticleColor()
  Multiply(
   A=LinearInterpolate(Alpha=Saturate(None=@n10)) {constA=0.3}
   B=DepthFade(Opacity=Multiply(A=@n19, B=@n20.A), FadeDistance=ScalarParameter() {defaultValue=44, parameterName=Depth Fade Distance})
  ) {constA=1}
  OneMinus(None=Fresnel(ExponentIn=@n1))
  ScalarParameter() {defaultValue=0.1, parameterName=Edge Fade}
  ScalarParameter() {defaultValue=0.6, parameterName=Length Fade}
  ScalarParameter() {defaultValue=0.7, parameterName=Tail Wave Speed}
  ScalarParameter() {defaultValue=0.1, parameterName=Tail Wave Intensity}
  n21 = Divide(A=@n11, B=ScalarParameter() {defaultValue=0.3, parameterName=Base Fade})
  VertexInterpolator(VS=Multiply(A=Saturate(None=@n21), B=@n19))
 ]
}
