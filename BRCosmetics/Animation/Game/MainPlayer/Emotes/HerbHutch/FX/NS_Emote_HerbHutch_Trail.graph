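# Niagara system for the HerbHutch emote, adding skeletal-mesh ribbon trails and fluttering confetti sprites.
# An anim-trail emitter builds three bone-pair trails, using per-group widths to split main and wing looks across two ribbon renderers.
# Confetti spawns at two bone or socket points, driven by gravity, curl noise and drag.
# User.LoopCount blends loop timing and trail fades, while time-of-day scalars adjust brightness and a front-end modifier adjusts trail opacity.
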NiagaraSystem /BRCosmetics/Animation/Game/MainPlayer/Emotes/HerbHutch/FX/NS_Emote_HerbHutch_Trail
dependencies {
 usedDataInterfaces [
  &NiagaraDataInterfaceArrayFloat
  &NiagaraDataInterfaceCurve
  &NiagaraDataInterfaceArrayColor
  &NiagaraDataInterfaceArrayFloat3
  &NiagaraDataInterfaceArrayInt32
  &NiagaraDataInterfaceSpriteRendererInfo
  &NiagaraDataInterfaceSkeletalMesh
 ]
 usedDynamicInputs []
 usedModules [
  &self:SystemSpawnScript.SystemScriptSource.SystemScriptGraph.NiagaraNodeAssignment_0.SetVariables_#1
  &SystemState
 ]
}
emitters {
 ConfettiSquares {
  emitterUpdateScript [
   &EmitterState(
    Life Cycle Mode=::NewEnumerator1
    Loop Behavior=::NewEnumerator1
    Loop Delay=&Lerp_Float(A=0.25, B=0, Alpha=$User.LoopCount)
    Loop Duration=3
    UseLoopDelay=true
   )
   &SpawnRate(SpawnRate=7, Use Spawn Probability=false)
  ]
  particleSpawnScript [
   &FN_SetSpawnGroup(Number of Spawn Groups=2)
   &InitializeParticle(
    Color=&SelectLinearColorFromArray(
     Array Sampling Mode=::NewEnumerator0
     Direct Array Index=?
     Color Selection Array=DataInterface() {propertyValues {ColorData [<rgba 0.8646 0.8646 0.8646 1>, <rgba 0.1006 0.1339 0.276 1>, <rgba 0.4583 0.1385 0.1385 1>]}}
     Random Seed=0
    )
    Lifetime Max=1.5
    Lifetime Mode=::NewEnumerator1
    Sprite Size Mode=::NewEnumerator1
    Uniform Sprite Size Max=5
    Uniform Sprite Size Min=2
    UsePositionOffset=false
    Write Material Random=true
   )
   &AddVelocity(
    Cone Angle=60
    Cone Axis=&SelectVectorFromArray(
     Direct Array Index=&SpawnGroup()
     Vector Selection Array=DataInterface() {propertyValues {FloatData [<-0.4 0 -1>, <0.4 0 -1>]}}
    )
    Module.UseConeFalloff=true
    Module.UseVelDistribution=true
    Override Seed=false
    Velocity Mode=::NewEnumerator2
    Velocity Speed=&RandomRangeFloat(Minimum=300, Maximum=1111)
   ) {enabled=false}
   &SkeletalMeshLocation(
    Bone / Socket Index=&SelectIntFromArray(Direct Array Index=&SpawnGroup(), Int Selection Array=DataInterface() {propertyValues {IntData [2,4]}})
    MaskSpawnGroup=false
    Mesh Sampling Type=::NewEnumerator5
    Sampled Position Offset=&RandomRangeVector(Minimum=<-3 -3 -3>, Maximum=<3 3 3>)
    Skeletal Mesh=$System.Skeletal Mesh
    Socket Sampling Mode=::NewEnumerator5
   )
  ]
  particleUpdateScript [
   &ParticleState()
   &GravityForce(Gravity=<0 0 -111>)
   /Niagara/Modules/Update/Forces/CurlNoiseForce(Noise Frequency=0.02, Noise Strength=200, Pan Noise=false)
   &Drag(
    Drag=&FloatFromCurve(FloatCurve=DataInterface() {propertyValues {Curve {keys [<at 0 0.5>,<at 1 1>]}}}, Scale Curve=6)
    Use Linear Drag=true
    Use Rotational Drag=true
   )
   &SolveForcesAndVelocity(Clamp Velocity=false, Limit Acceleration=false)
   &ScaleSpriteSize(Scale Factor=&Vector2DFromFloat(Value=&FloatFromCurve(FloatCurve=DataInterface() {propertyValues {Curve {keys [
          <at 0 0 RCIM_Cubic>
          {arriveTangent=-0.0623, interpMode=RCIM_Cubic, leaveTangent=-0.0623, tangentMode=RCTM_User, time=0.04, value=1}
          {arriveTangent=-0.0083, interpMode=RCIM_Cubic, leaveTangent=-0.0083, tangentMode=RCTM_User, time=0.5, value=1}
          <at 1 0 RCIM_Cubic>
         ]}}})))
   &SubUVAnimation(
    "Sprite Renderer "=DataInterface() {propertyValues {SpriteRenderer=&self:ConfettiSquares_0.NiagaraSpriteRendererProperties_0}}
    Start Frame Offset:NiagaraInt32=&RandomRangeInt(Minimum=0, Maximum=555)
    UseEndFrame=false
    UseStartFrame=false
   )
   &ScaleColor(
    Scale RGB=&VectorFromFloat(Value=&FNTimeOfDayScalar(Night Scale=0.25, Morning Scale=0.333, Day Scale=0.5, Evening Scale=0.25))
    ScaleA=true
    ScaleRGB=true
   )
  ]
  rendererClasses [&NiagaraSpriteRendererProperties]
  renderers [{rendererClass=&NiagaraSpriteRendererProperties}]
 }
 FN_AnimTrail002 {
  emitterSpawnScript [&FN_RibbonIDOffsetUtility()]
  emitterUpdateScript [
   &EmitterState(
    Inactive Response=::NewEnumerator1
    Life Cycle Mode=::NewEnumerator1
    Loop Behavior=::NewEnumerator1
    Loop Duration=&Lerp_Float(A=1.6, B=1.9, Alpha=$User.LoopCount)
    UseLoopDelay=false
   )
   &FN_SpawnAnimTrail(Number of Anim Trails=3)
  ]
  particleSpawnScript [
   &FN_SetSpawnGroup(Number of Spawn Groups=3)
   &FN_InitializeAnimTrailBoneIndecies(First Bone or Socket Index=1, Second Bone or Socket Index=0)
   &FN_InitializeAnimTrailBoneIndecies(AnimTrail Selection=1, First Bone or Socket Index=3, Second Bone or Socket Index=2) {module=FN_InitializeAnimTrailBoneIndecies001}
   &FN_InitializeAnimTrailBoneIndecies(AnimTrail Selection=2, First Bone or Socket Index=5, Second Bone or Socket Index=4) {module=FN_InitializeAnimTrailBoneIndecies002}
   &FN_InitializeAnimTrail(
    Lifetime=&SelectFloatFromArray(
     Direct Array Index=&SpawnGroup()
     Float Selection Array=DataInterface() {propertyValues {FloatData [0.42,0.2,0.2]}}
    )
    SkeletalMesh=$System.Skeletal Mesh
   )
   &AddVelocity(Velocity=<0 400 0>) {enabled=false}
   &DynamicMaterialParameters(
    Index 0 Param 1=1
    Index 0 Param 2=0.3
    Index 0 Param 3=0.4
    Index 0 Param 4=0.6
    Module.Param0WriteEnabled=true
    Module.Param1WriteEnabled=true
    Module.Param2WriteEnabled=true
    Module.Param3WriteEnabled=true
   ) {module=DynamicMaterialParameters002}
   /SetVariables(
    Particles.RibbonWidthMain:NiagaraFloat=&Multiply_Float(
     A=$Particles.RibbonWidth
     B=&SelectFloatFromArray(Direct Array Index=&SpawnGroup(), Float Selection Array=DataInterface() {propertyValues {FloatData [1,0,0]}})
    )
    Particles.RibbonWidthWings:NiagaraFloat=&Multiply_Float(
     A=$Particles.RibbonWidth
     B=&SelectFloatFromArray(Direct Array Index=&SpawnGroup(), Float Selection Array=DataInterface() {propertyValues {FloatData [0,1,1]}})
    )
   ) {isSetParameters=true}
  ]
  particleUpdateScript [
   &InheritSourceMovement(Applied Position Scale=<0.8 0.8 0.8>, ApplyForce=false, ApplyPosition=true, ApplyPositionFalloff=false)
   &SolveForcesAndVelocity(Clamp Velocity=false, Limit Acceleration=false)
   &FN_UpdateAnimTrail(SkeletalMesh=$System.Skeletal Mesh) {module=FN_UpdateAnimTrail001}
   &ScaleColor(
    Scale Alpha=&FloatFromCurve(
     FloatCurve=DataInterface() {propertyValues {Curve {keys [{arriveTangent=-3.3248, interpMode=RCIM_Cubic, time=0, value=1}, {interpMode=RCIM_Cubic, leaveTangent=-9.9298, time=1, value=0}]}}}
     Scale Curve=&Multiply_Float(
      A=&FN_FrontEndFloatModifier(Gameplay=1.8, Front End=1.2)
      B=&Lerp_Float(
       A=&FloatFromCurve(
        FloatCurve=DataInterface() {propertyValues {Curve {keys [<at 1.4 1 RCIM_Cubic>,<at 1.6 0>]}}}
        CurveIndex=$Emitter.Age
        Scale Curve=&FNTimeOfDayScalar(Night Scale=0.8, Morning Scale=0.9, Evening Scale=0.9)
       )
       B=&FloatFromCurve(
        FloatCurve=DataInterface() {propertyValues {Curve {keys [<at 1.45 1 RCIM_Cubic>,<at 1.75 0>]}}}
        CurveIndex=$Emitter.Age
        Scale Curve=&FNTimeOfDayScalar(Night Scale=0.8, Morning Scale=0.9, Evening Scale=0.9)
       )
       Alpha=$User.LoopCount
      )
     )
    )
    Scale RGB=&VectorFromFloat(Value=&FNTimeOfDayScalar(Night Scale=0.4, Morning Scale=0.8, Evening Scale=0.75))
    ScaleA=true
    ScaleRGB=true
   )
   /SetVariables(Particles.WingsColor:LinearColor=&MakeLinearColorFromVectorAndFloat(
     "Vector (RGB)"=&Multiply_VectorByFloat(
      Vector=<0.0835 0.1111 0.2292>
      Float=&FNTimeOfDayScalar(Night Scale=0.6, Morning Scale=0.8, Evening Scale=0.8)
     )
     "Float (Alpha)"=&FloatFromCurve(
      FloatCurve=DataInterface() {propertyValues {Curve {keys [<at 0 0 RCIM_Cubic>, <at 0.15 1 RCIM_Cubic>, {interpMode=RCIM_Cubic, leaveTangent=-9.9298, time=1, value=0}]}}}
      Scale Curve=&Multiply_Float(
       A=&FN_FrontEndFloatModifier(Gameplay=1.05, Front End=0.7)
       B=&Lerp_Float(
        A=&FloatFromCurve(
         FloatCurve=DataInterface() {propertyValues {Curve {keys [<at 1.4 1 RCIM_Cubic>,<at 1.6 0>]}}}
         CurveIndex=$Emitter.Age
         Scale Curve=&FNTimeOfDayScalar(Night Scale=0.7, Morning Scale=0.85, Evening Scale=0.85)
        )
        B=&FloatFromCurve(
         FloatCurve=DataInterface() {propertyValues {Curve {keys [<at 1.6 1 RCIM_Cubic>,<at 1.9 0>]}}}
         CurveIndex=$Emitter.Age
         Scale Curve=&FNTimeOfDayScalar(Night Scale=0.7, Morning Scale=0.85, Evening Scale=0.85)
        )
        Alpha=$User.LoopCount
       )
      )
     )
    )) {isSetParameters=true}
  ]
  rendererClasses [&NiagaraRibbonRendererProperties]
  renderers [
   {rendererClass=&NiagaraRibbonRendererProperties}
   {rendererClass=&NiagaraRibbonRendererProperties, rendererIndex=1}
  ]
 }
}
summary {
 systemName=NS_Emote_HerbHutch_Trail
 emitters [
  {emitterName=FN_AnimTrail002, rendererClasses [&NiagaraRibbonRendererProperties]}
  {emitterName=ConfettiSquares, rendererClasses [&NiagaraSpriteRendererProperties]}
 ]
 userVariables [
  User.bIsFrontEnd {type=NiagaraBool}
  User.bIsFrontEndPreview {type=NiagaraBool}
  User.LoopCount {type=NiagaraFloat, defaultValue=0, description=""}
  User.RibbonIDOffset {type=NiagaraInt32}
 ]
}
